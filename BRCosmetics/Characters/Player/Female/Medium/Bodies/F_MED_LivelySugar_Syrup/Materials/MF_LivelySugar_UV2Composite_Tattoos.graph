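# Material function for the LivelySugar character body that composites a UV2 layer, evidently tattoos, onto incoming material attributes, with per-body-region hiding.
# Static switches gate the whole composite and separate diffuse, normal and SRM blends, all sampled on the second UV set.
# The blend alpha comes from diffuse alpha, a mask texture channel or a TechArtMask input, scaled by strength.
# Five arm, torso and leg region masks, weighted by TattooHide scalars, subtract from that alpha.
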MaterialFunction /BRCosmetics/Characters/Player/Female/Medium/Bodies/F_MED_LivelySugar_Syrup/Materials/MF_LivelySugar_UV2Composite_Tattoos
graph {nodes [
  n1 = MaterialFunctionCall(Float3=NamedRerouteUsage()) {materialFunction=&BreakOutFloat3Components}
  n2 = SetMaterialAttributes!(
   MaterialAttributes=NamedRerouteUsage()
   BaseColor=NamedRerouteUsage()
   Normal=NamedRerouteUsage()
   Specular=@n1.R
   Roughness=@n1.R
   Metallic=@n1.R
  ) {attributeSetTypes [
    #1
    #2
    #3
    #4
    #5
   ]}
  FunctionOutput(None=StaticSwitchParameter(True=@n2, False=NamedRerouteUsage()) {defaultValue=true, group=UV2 Composite, parameterName=UseUV2Composite, sortPriority=0})
  n3 = GetMaterialAttributes(None=NamedRerouteDeclaration(None=FunctionInput() {inputName=Material, inputType=FunctionInput_MaterialAttributes}) {name=IN_Material, nodeColor=<rgba 1 0.0234 0 1>}) {attributeGetTypes [
    #1
    #2
    #3
    #4
    #5
   ]}
  NamedRerouteDeclaration(None=@n3.BaseColor) {name=IN_Diffuse, nodeColor=<rgba 0 0.2969 1 1>}
  n4 = TextureSample() {
   constCoordinate=1
   samplerType=SAMPLERTYPE_Masks
   texture=/BRCosmetics/Characters/Player/Female/Medium/Bodies/F_MED_LivelySugar_Shore/Textures/T_LivelySugar_Syrup_TattooRegion1_FX
  }
  NamedRerouteDeclaration(None=@n4.G) {name=Region_3_leftArm, nodeColor=<rgba 0.4062 1 0 1>}
  NamedRerouteDeclaration(None=@n4.B) {name=Region_1_rightArm, nodeColor=<rgba 1 0.1491 0.5686 1>}
  n5 = TextureSample() {
   constCoordinate=1
   samplerType=SAMPLERTYPE_Masks
   texture=/BRCosmetics/Characters/Player/Female/Medium/Bodies/F_MED_LivelySugar_Shore/Textures/T_LivelySugar_Syrup_TattooRegion2_FX
  }
  NamedRerouteDeclaration(None=@n5.R) {name=Region_4_rightLeg, nodeColor=<rgba 0.2442 0.7768 1 1>}
  NamedRerouteDeclaration(None=@n5.G) {name=Region_5_leftLeg, nodeColor=<rgba 1 0.687 0.0901 1>}
  NamedRerouteDeclaration(None=@n3.Normal) {name=IN_Normal, nodeColor=<rgba 0 0.2969 1 1>}
  NamedRerouteDeclaration(None=MaterialFunctionCall(X=@n3.Specular, Y=@n3.Specular, Z=@n3.Specular) {materialFunction=&MakeFloat3}.Result) {name=IN_SRM, nodeColor=<rgba 0 0.2969 1 1>}
  n6 = NamedRerouteUsage()
  n7 = TextureSampleParameter2D(UVs=TextureCoordinate() {coordinateIndex=1}) {
   group=UV2 Composite
   parameterName=UV2Composite_Normals
   samplerType=SAMPLERTYPE_Normal
   sortPriority=21
   texture=&FlatNormalBlue
  }
  n8 = NamedRerouteUsage()
  NamedRerouteDeclaration(None=StaticSwitchParameter(True=LinearInterpolate(A=@n6, B=@n7.RGB, Alpha=@n8), False=NamedRerouteUsage()) {group=UV2 Composite, parameterName=UseUV2Normals, sortPriority=20}) {name=OUT_Normal, nodeColor=<rgba 1 0 0.7969 1>}
  n9 = NamedRerouteUsage()
  n10 = TextureSampleParameter2D(UVs=TextureCoordinate() {coordinateIndex=1}) {group=UV2 Composite, parameterName=UV2Composite_Diffuse, sortPriority=11, texture=/Game/Global/Textures/Default/Blanks/T_MidgreysRGB}
  n11 = NamedRerouteUsage()
  NamedRerouteDeclaration(None=StaticSwitchParameter(True=LinearInterpolate(A=@n9, B=@n10.RGB, Alpha=@n11), False=NamedRerouteUsage()) {defaultValue=true, group=UV2 Composite, parameterName=UseUV2Diffuse, sortPriority=10}) {name=OUT_Diffuse, nodeColor=<rgba 1 0 0.7969 1>}
  n12 = AppendVector(A=FunctionInput() {bUsePreviewValueAsDefault=true, inputName=TechArtMask, sortPriority=1}, B=Constant() {r=1})
  n13 = TextureSampleParameter2D(UVs=StaticSwitchParameter(True=TextureCoordinate(), False=TextureCoordinate() {coordinateIndex=1}) {group=UV2 Composite, parameterName=UV2Composite_AlphaTextureUseUV1, sortPriority=4}) {
   group=UV2 Composite
   parameterName=UV2Composite_AlphaTexture
   samplerType=SAMPLERTYPE_Masks
   sortPriority=3
   texture=&T_BlackMask
  }
  n14 = StaticSwitchParameter(
   True=NamedRerouteUsage()
   False=ChannelMaskParameter(None=StaticSwitchParameter(True=@n12, False=@n13.RGBA) {group=UV2 Composite, parameterName=UseTechArtMaskAsAlpha, sortPriority=2}) {group=UV2 Composite, parameterName=UV2Composite_AlphaChannel, sortPriority=4}.R
  ) {desc="If true, uses the diffuse alpha channel as the alpha mask", group=UV2 Composite, parameterName=UseDiffuseAlphaChannel, sortPriority=1}
  n15 = Multiply(A=@n14, B=ScalarParameter() {defaultValue=1, group=UV2 Composite, parameterName=UV2Composite_AlphaStrength, sortPriority=5})
  n16 = Add(
   A=Multiply(A=NamedRerouteUsage(), B=ScalarParameter() {group=LivelySugar Tattoos, parameterName=TattooHide_Region_1})
   B=Multiply(A=NamedRerouteUsage(), B=ScalarParameter() {group=LivelySugar Tattoos, parameterName=TattooHide_Region_2})
  )
  n17 = Multiply(A=NamedRerouteUsage(), B=ScalarParameter() {group=LivelySugar Tattoos, parameterName=TattooHide_Region_3})
  n18 = NamedRerouteUsage()
  n19 = ScalarParameter() {group=LivelySugar Tattoos, parameterName=TattooHide_Region_4}
  n20 = Add(
   A=Add(A=Add(A=@n16, B=@n17), B=Multiply(A=@n18, B=@n19))
   B=Multiply(A=NamedRerouteUsage(), B=ScalarParameter() {group=LivelySugar Tattoos, parameterName=TattooHide_Region_5})
  )
  NamedRerouteDeclaration(None=Saturate(None=Subtract(A=@n15, B=@n20))) {name=UV2Alpha, nodeColor=<rgba 0 0.4609 1 1>}
  n21 = NamedRerouteUsage()
  n22 = TextureSampleParameter2D(UVs=TextureCoordinate() {coordinateIndex=1}) {
   group=UV2 Composite
   parameterName=UV2Composite_SRM
   samplerType=SAMPLERTYPE_Masks
   sortPriority=31
   texture=&T_Fortnite_Default_S
  }
  n23 = NamedRerouteUsage()
  NamedRerouteDeclaration(None=StaticSwitchParameter(True=LinearInterpolate(A=@n21, B=@n22.RGB, Alpha=@n23), False=NamedRerouteUsage()) {group=UV2 Composite, parameterName=UseUV2SRM, sortPriority=30}) {name=OUT_SRM, nodeColor=<rgba 1 0 0.7969 1>}
  NamedRerouteDeclaration(None=@n4.R) {name=Region_2_torso, nodeColor=<rgba 0 0.0859 1 1>}
  NamedRerouteDeclaration(None=@n10.A) {name=DiffuseAlpha, nodeColor=<rgba 1 0.5391 0 1>}
 ]}
