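# AnimBlueprint for the F_MED_LivelySugar_Syrup body cosmetic, adding secondary motion to outfit and chest bones.
# The parent mesh pose runs through Control Rig, AnimDynamics, bone offsets, spine-weighted constraints and a RigidBody using the Shore variant's physics asset.
# The cached result optionally gains layered angular and linear chest springs via bPbBlendEnabled.
# Each update computes gravity override from Shore parameters, belt simulation alpha from movement state, and the rig's parachute value.
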AnimBlueprint /BRCosmetics/Characters/Player/Female/Medium/Bodies/F_MED_LivelySugar_Syrup/Meshes/F_MED_LivelySugar_Syrup_AnimBP
target_skeleton=&/F_MED_LivelySugar_Syrup_Skeleton
graphs [
 AnimGraph {nodes [
   n1 = "Transform (Modify) Bone - Bone: dyn_chest_l"(ComponentPose=Local To Component(LocalPose=Use cached pose 'allDynCache'()), Alpha=1, Translation="0,0,1", Rotation="0,0,0", Scale="1,1,1") {translationMode=BMM_Additive, translationSpace=BCS_BoneSpace, boneToModify {boneName=dyn_chest_l}, translation {z=1}}
   n2 = "Transform (Modify) Bone - Bone: dyn_chest_r"(ComponentPose=@n1, Alpha=1, Translation="0,0,1", Rotation="0,0,0", Scale="1,1,1") {translationMode=BMM_Additive, translationSpace=BCS_BoneSpace, boneToModify {boneName=dyn_chest_r}, translation {z=1}}
   n3 = Get Sim Space Settings() -> (Sim Space Settings:Anim Phys Sim Space Settings Structure)
   n4 = Anim Dynamics - Bone: dyn_chest_r!(ComponentPose=@n2, Alpha=1, AngularSpringConstant:float=5000, SimSpaceSettings:Anim Phys Sim Space Settings Structure=@n3) -> (Pose) {
    angularBiasOverride=1
    angularDampingOverride=0.9999
    angularSpringConstant=5000
    bAngularSpring=true
    bGravityOverrideInSimSpace=true
    bOverrideAngularBias=true
    bOverrideAngularDamping=true
    bUseGravityOverride=true
    linearDampingOverride=0.9999
    linearSpringConstant=100
    simulationSpace=RootRelative
    boundBone {boneName=dyn_chest_r}
    componentAppliedLinearAccClamp=<2 2 2>
    gravityOverride {z=-980}
    physicsBodyDefinitions [{
      boundBone {boneName=dyn_chest_r}
      boxExtents=<21 21 21>
      constraintSetup {angularLimitsMax=<0 1 5>, angularLimitsMin=<0 -5 -5>, angularTarget=<1 0 0>}
      localJointOffset=<2 0 0>
     }]
    relativeSpaceBone {boneName=spine_04}
    simSpaceSettings {maxAngularAcceleration=45, maxAngularVelocity=8, simSpaceAngularAlpha=0.25}
   }
   n5 = Anim Dynamics - Bone: dyn_chest_l!(ComponentPose=@n4, Alpha=1, AngularSpringConstant:float=5000, SimSpaceSettings:Anim Phys Sim Space Settings Structure=@n3) -> (Pose) {
    angularBiasOverride=1
    angularDampingOverride=0.9999
    angularSpringConstant=5000
    bAngularSpring=true
    bGravityOverrideInSimSpace=true
    bOverrideAngularBias=true
    bOverrideAngularDamping=true
    bUseGravityOverride=true
    linearDampingOverride=0.9999
    linearSpringConstant=100
    simulationSpace=RootRelative
    boundBone {boneName=dyn_chest_l}
    componentAppliedLinearAccClamp=<2 2 2>
    gravityOverride {z=-980}
    physicsBodyDefinitions [{
      boundBone {boneName=dyn_chest_l}
      boxExtents=<20 20 20>
      constraintSetup {angularLimitsMax=<0 1 5>, angularLimitsMin=<0 -5 -5>, angularTarget=<1 0 0>}
      localJointOffset=<2 0 0>
     }]
    relativeSpaceBone {boneName=spine_04}
    simSpaceSettings {maxAngularAcceleration=45, maxAngularVelocity=8, simSpaceAngularAlpha=0.25}
   }
   n6 = Anim Dynamics - Bone: dyn_chest_r!(ComponentPose=@n5, Alpha=1, GravityOverride:Vector="0,0,-200", LinearSpringConstant:float=180) -> (Pose) {
    angularBiasOverride=1
    angularDampingOverride=0.99
    angularSpringConstant=10000
    bGravityOverrideInSimSpace=true
    bLinearSpring=true
    bOverrideAngularBias=true
    bOverrideLinearDamping=true
    bUseGravityOverride=true
    linearDampingOverride=1
    linearSpringConstant=180
    simulationSpace=RootRelative
    boundBone {boneName=dyn_chest_r}
    componentAppliedLinearAccClamp=<5 5 5>
    gravityOverride {z=-200}
    physicsBodyDefinitions [{
      boundBone {boneName=dyn_chest_r}
      boxExtents=<80 80 80>
      constraintSetup {angularLimitsMax=<0 0 0>, angularLimitsMin=<0 0 0>, angularTarget=<1 0 0>, linearAxesMax=<0 0.5 1>, linearAxesMin=<0 -0.5 -0.5>}
      localJointOffset=<0.2 0 0>
     }]
    relativeSpaceBone {boneName=spine_04}
    simSpaceSettings {maxAngularAcceleration=50, maxAngularVelocity=10, simSpaceAngularAlpha=0.1}
   }
   n7 = Anim Dynamics - Bone: dyn_chest_l!(ComponentPose=@n6, Alpha=1, GravityOverride:Vector="0,0,-200", LinearSpringConstant:float=180) -> (Pose) {
    angularBiasOverride=1
    angularDampingOverride=0.99
    angularSpringConstant=10000
    bGravityOverrideInSimSpace=true
    bLinearSpring=true
    bOverrideAngularBias=true
    bOverrideLinearDamping=true
    bUseGravityOverride=true
    linearDampingOverride=1
    linearSpringConstant=180
    simulationSpace=RootRelative
    boundBone {boneName=dyn_chest_l}
    componentAppliedLinearAccClamp=<5 5 5>
    gravityOverride {z=-200}
    physicsBodyDefinitions [{
      boundBone {boneName=dyn_chest_l}
      boxExtents=<90 90 90>
      constraintSetup {angularLimitsMax=<0 0 0>, angularLimitsMin=<0 0 0>, angularTarget=<1 0 0>, linearAxesMax=<0 0.5 1>, linearAxesMin=<0 -0.5 -0.5>}
      localJointOffset=<0.2 0 0>
     }]
    relativeSpaceBone {boneName=spine_04}
    simSpaceSettings {maxAngularAcceleration=50, maxAngularVelocity=10, simSpaceAngularAlpha=0.1}
   }
   Output Pose(Result=Blend Poses by bool(
     BlendPose_0=Component To Local(ComponentPose=@n7)
     BlendPose_1=Use cached pose 'allDynCache'()
     bActiveValue=Get bPbBlendEnabled() -> (bPbBlendEnabled:bool)
    )) {layerGroup=None}
   n8 = Control Rig!(
    Source=Copy Pose From Mesh() {bUseAttachedParent=true}
    Alpha
    isSkydiving_CR:bool=Get bIsSkyDiving()
    isKneeSliding_CR:bool=Get bIsKneeSliding()
    isCrouching_CR:bool=Get bIsCrouching()
    isFalling_CR:double=Get isFallingNotMoving() -> (isFallingNotMoving:double)
    Parachuting float:double="Get Parachuting shoulder value "() -> ("Parachuting shoulder value ":double)
   ) -> (Pose) {controlRigAssetReference {blueprintRigClass=&/F_MED_LivelySugar_Syrup_CtrlRig.F_MED_LivelySugar_Syrup_CtrlRig_C}}
   n9 = Get isFallingNotMoving()
   n10 = Anim Dynamics - Bone: dyn_belt01!(ComponentPose=Local To Component(LocalPose=@n8), Alpha=@n9, GravityOverride:Vector="0,0,-980") -> (Pose) {
    angularDampingOverride=1
    angularSpringConstant=20000
    bOverrideAngularDamping=true
    bUseGravityOverride=true
    gravityScale=0.35
    simulationSpace=RootRelative
    alphaBoolBlend {blendInTime=0.1, blendOption=Linear, blendOutTime=0.2}
    alphaScaleBiasClamp {bInterpResult=true}
    boundBone {boneName=dyn_belt01}
    componentAppliedLinearAccClamp=<4000 4000 3500>
    componentLinearAccScale=<1 1 0.6>
    componentLinearVelScale=<1 1 0.6>
    gravityOverride {z=-980}
    physicsBodyDefinitions [{
      boundBone {boneName=dyn_belt01}
      boxExtents=<28 28 28>
      constraintSetup {angularLimitsMax=<0 7.5 1>, angularLimitsMin=<0 0 0>, angularTarget=<1 0 0>}
      localJointOffset=<2 0 0>
     }]
    relativeSpaceBone {boneName=pelvis}
    simSpaceSettings {simSpaceAngularAlpha=0.1}
   }
   n11 = Anim Dynamics - Bone: dyn_belt02!(ComponentPose=@n10, Alpha=@n9, GravityOverride:Vector="0,0,-980") -> (Pose) {
    angularDampingOverride=1
    angularSpringConstant=20000
    bOverrideAngularDamping=true
    bUseGravityOverride=true
    gravityScale=0.35
    simulationSpace=RootRelative
    alphaBoolBlend {blendInTime=0.1, blendOption=Linear, blendOutTime=0.2}
    alphaScaleBiasClamp {bInterpResult=true}
    boundBone {boneName=dyn_belt02}
    componentAppliedLinearAccClamp=<4000 4000 3500>
    componentLinearAccScale=<1 1 0.6>
    componentLinearVelScale=<1 1 0.6>
    gravityOverride {z=-980}
    physicsBodyDefinitions [{
      boundBone {boneName=dyn_belt02}
      boxExtents=<28 28 28>
      constraintSetup {angularLimitsMax=<0 6.5 1>, angularLimitsMin=<0 0 0>, angularTarget=<1 0 0>}
      localJointOffset=<2 0 0>
     }]
    relativeSpaceBone {boneName=pelvis}
    simSpaceSettings {simSpaceAngularAlpha=0.1}
   }
   n12 = Anim Dynamics - Bone: dyn_belt03!(ComponentPose=@n11, Alpha=@n9, GravityOverride:Vector="0,0,-980") -> (Pose) {
    angularDampingOverride=1
    angularSpringConstant=20000
    bOverrideAngularDamping=true
    bUseGravityOverride=true
    gravityScale=0.35
    simulationSpace=RootRelative
    alphaBoolBlend {blendInTime=0.1, blendOption=Linear, blendOutTime=0.2}
    alphaScaleBiasClamp {bInterpResult=true}
    boundBone {boneName=dyn_belt03}
    componentAppliedLinearAccClamp=<4000 4000 3500>
    componentLinearAccScale=<1 1 0.6>
    componentLinearVelScale=<1 1 0.6>
    gravityOverride {z=-980}
    physicsBodyDefinitions [{
      boundBone {boneName=dyn_belt03}
      boxExtents=<28 28 28>
      constraintSetup {angularLimitsMax=<0 5 0>, angularLimitsMin=<0 0 0>, angularTarget=<1 0 0>}
      localJointOffset=<2 0 0>
     }]
    relativeSpaceBone {boneName=pelvis}
    simSpaceSettings {simSpaceAngularAlpha=0.1}
   }
   n13 = "Transform (Modify) Bone - Bone: base_holster"!(ComponentPose=@n12, bAlphaBoolEnabled:bool=True, Translation="-1.936,0.000,-0.300", Rotation="0,0,0", Scale="1,1,1") -> (Pose) {
    alphaInputType=Bool
    rotationMode=BMM_Additive
    rotationSpace=BCS_BoneSpace
    translationMode=BMM_Additive
    translationSpace=BCS_BoneSpace
    boneToModify {boneName=base_holster}
    translation {x=-1.936, z=-0.3}
   }
   n14 = "Transform (Modify) Bone - Bone: base_belt_l"!(ComponentPose=@n13, bAlphaBoolEnabled:bool=True, Translation="-1.872,-2.148,0.000", Rotation="0,0,0", Scale="1,1,1") -> (Pose) {
    alphaInputType=Bool
    rotationMode=BMM_Additive
    rotationSpace=BCS_BoneSpace
    translationMode=BMM_Additive
    translationSpace=BCS_BoneSpace
    alphaBoolBlend {blendInTime=5, blendOutTime=0.6}
    boneToModify {boneName=base_belt_l}
    translation=<-1.872 -2.148>
   }
   n15 = Anim Dynamics - Bone: dyn_holster(ComponentPose=@n14, Alpha=1, ExternalForce="0,0,0") {
    angularDampingOverride=1
    angularSpringConstant=17000
    bAngularSpring=true
    bOverrideAngularDamping=true
    bUseGravityOverride=true
    gravityScale=0.35
    simulationSpace=RootRelative
    boundBone {boneName=dyn_holster}
    componentAppliedLinearAccClamp=<3000 3000 3000>
    componentLinearAccScale=<0.3 0.3 0.3>
    componentLinearVelScale=<0.3 0.3 0.3>
    gravityOverride {z=-980}
    physicsBodyDefinitions [{
      boundBone {boneName=dyn_holster}
      boxExtents=<35 35 35>
      constraintSetup {angularLimitsMax=<0 0 6>, angularLimitsMin=<0 -5 -6>, angularTarget=<1 0 0>}
      localJointOffset=<2 0 0>
     }]
   }
   n16 = Anim Dynamics - Bone: dyn_heelBullet_l!(ComponentPose=@n15, Alpha=1, GravityOverride:Vector="0,0,-980", ExternalForce="0,0,0") -> (Pose) {
    angularDampingOverride=0.999
    angularSpringConstant=6000
    bAngularSpring=true
    bOverrideAngularDamping=true
    bUseGravityOverride=true
    gravityScale=0.35
    simulationSpace=BoneRelative
    boundBone {boneName=dyn_heelBullet_l}
    componentAppliedLinearAccClamp=<3000 3000 3000>
    componentLinearAccScale=<0.5 0.5 0.5>
    componentLinearVelScale=<0.5 0.5 0.5>
    gravityOverride {z=-980}
    physicsBodyDefinitions [{
      boundBone {boneName=dyn_heelBullet_l}
      boxExtents=<15 15 15>
      constraintSetup {angularLimitsMax=<0 40 0>, angularLimitsMin=<0 -40 -60>, angularTarget=<1 0 0>}
      localJointOffset=<3 -0.5 0.5>
     }]
    relativeSpaceBone {boneName=heelFoot_l}
    simSpaceSettings {simSpaceAngularAlpha=0.1}
   }
   n17 = Anim Dynamics - Bone: dyn_heelBullet_r!(ComponentPose=@n16, Alpha=1, GravityOverride:Vector="0,0,-980", ExternalForce="0,0,0") -> (Pose) {
    angularDampingOverride=0.999
    angularSpringConstant=6000
    bAngularSpring=true
    bOverrideAngularDamping=true
    bUseGravityOverride=true
    gravityScale=0.35
    simulationSpace=BoneRelative
    boundBone {boneName=dyn_heelBullet_r}
    componentAppliedLinearAccClamp=<3000 3000 3000>
    componentLinearAccScale=<0.5 0.5 0.5>
    componentLinearVelScale=<0.5 0.5 0.5>
    gravityOverride {z=-980}
    physicsBodyDefinitions [{
      boundBone {boneName=dyn_heelBullet_r}
      boxExtents=<15 15 15>
      constraintSetup {angularLimitsMax=<0 40 0>, angularLimitsMin=<0 -40 -60>, angularTarget=<1 0 0>}
      localJointOffset=<3 -0.5 0.5>
     }]
    relativeSpaceBone {boneName=heelFoot_r}
    simSpaceSettings {simSpaceAngularAlpha=0.1}
   }
   n18 = Anim Dynamics - Bone: dyn_bracelet(ComponentPose=@n17, Alpha=1, ExternalForce="0,0,0") {
    angularDampingOverride=1
    angularSpringConstant=15000
    bAngularSpring=true
    bOverrideAngularDamping=true
    bUseGravityOverride=true
    gravityScale=0.35
    simulationSpace=RootRelative
    boundBone {boneName=dyn_bracelet}
    componentAppliedLinearAccClamp=<6000 6000 6000>
    componentLinearAccScale=<0.7 0.7 0.9>
    componentLinearVelScale=<0.7 0.7 0.9>
    gravityOverride {z=-980}
    physicsBodyDefinitions [{
      boundBone {boneName=dyn_bracelet}
      boxExtents=<29 29 29>
      constraintSetup {angularLimitsMax=<0 5 12>, angularLimitsMin=<0 -12 -11>, angularTarget=<1 0 0>}
      localJointOffset=<4 0 0>
     }]
   }
   n19 = "Transform (Modify) Bone - Bone: dyn_bracelet"!(ComponentPose=@n18, bAlphaBoolEnabled:bool=True, Translation="0,0,0", Rotation="16,-7,0", Scale="1,1,1") -> (Pose) {
    alphaInputType=Bool
    rotationMode=BMM_Additive
    rotationSpace=BCS_BoneSpace
    translationMode=BMM_Additive
    translationSpace=BCS_BoneSpace
    alphaBoolBlend {blendInTime=0.1, blendOutTime=0.1}
    boneToModify {boneName=dyn_bracelet}
    rotation {pitch=16, yaw=-7}
   }
   n20 = RigidBody!(ComponentPose=@n19, Alpha, OverrideWorldGravity:Vector="0,0,0", ExternalForce) -> (Pose) {
    bOverrideWorldGravity=true
    overridePhysicsAsset=/BRCosmetics/Characters/Player/Female/Medium/Bodies/F_MED_LivelySugar_Shore/Meshes/F_MED_LivelySugar_Shore_Physics
    simulationSpace=BaseBoneSpace
    baseBoneRef {boneName=pelvis}
    componentAppliedLinearAccClamp=<1000 1000 1000>
    simSpaceSettings {
     maxAngularAcceleration=45
     maxAngularVelocity=8
     maxLinearAcceleration=1000
     maxLinearVelocity=300
     worldAlpha=1
     externalLinearDragV=<1 1 0.8>
    }
   }
   n21 = "Transform (Modify) Bone - Bone: base_handkerchiefBottom"!(ComponentPose=@n20, bAlphaBoolEnabled:bool=True, Translation="0,0,0", Rotation="-0.000,-0.022,5.273", Scale="1,1,1") -> (Pose) {
    alphaInputType=Bool
    rotationMode=BMM_Additive
    rotationSpace=BCS_BoneSpace
    translationMode=BMM_Additive
    translationSpace=BCS_BoneSpace
    boneToModify {boneName=base_handkerchiefBottom}
    rotation=<rot -0.0001 -0.022 5.273>
   }
   n22 = "Transform (Modify) Bone - Bone: base_handkerchiefTop"!(ComponentPose=@n21, bAlphaBoolEnabled:bool=True, Translation="0,0,0", Rotation="4.908,-9.467,11.603", Scale="1,1,1") -> (Pose) {
    alphaInputType=Bool
    rotationMode=BMM_Additive
    rotationSpace=BCS_BoneSpace
    translationMode=BMM_Additive
    translationSpace=BCS_BoneSpace
    boneToModify {boneName=base_handkerchiefTop}
    rotation=<rot 4.908 -9.4669 11.603>
   }
   n23 = "Constraint - base_bow_l (PPPPPP)"!(
    ComponentPose=@n22
    Alpha=1
    ConstraintWeights_0=0.2901
    ConstraintWeights_1:float=0.2463
    ConstraintWeights_2:float=0.2386
    ConstraintWeights_3:float=0.1524
    ConstraintWeights_4:float=0.0644
    ConstraintWeights_5:float=0.0082
   ) -> (Pose) {
    boneToModify {boneName=base_bow_l}
    constraintSetup [
     {targetBone {boneName=spine_04}}
     {targetBone {boneName=spine_03}}
     {targetBone {boneName=spine_02}}
     {targetBone {boneName=spine_05}}
     {targetBone {boneName=spine_01}}
     {targetBone {boneName=pelvis}}
    ]
    constraintWeights [1,1,1,1,1,1]
   }
   n24 = "Constraint - base_bow_r (PPPPPP)"!(
    ComponentPose=@n23
    Alpha=1
    ConstraintWeights_0=0.2898
    ConstraintWeights_1:float=0.247
    ConstraintWeights_2:float=0.2383
    ConstraintWeights_3:float=0.1526
    ConstraintWeights_4:float=0.0642
    ConstraintWeights_5:float=0.008
   ) -> (Pose) {
    boneToModify {boneName=base_bow_r}
    constraintSetup [
     {targetBone {boneName=spine_04}}
     {targetBone {boneName=spine_03}}
     {targetBone {boneName=spine_02}}
     {targetBone {boneName=spine_05}}
     {targetBone {boneName=spine_01}}
     {targetBone {boneName=pelvis}}
    ]
    constraintWeights [1,1,1,1,1,1]
   }
   n25 = "Constraint - base_cord_l_01 (PPPPPP)"!(
    ComponentPose=@n24
    Alpha=1
    ConstraintWeights_0=0.2837
    ConstraintWeights_1:float=0.256
    ConstraintWeights_2:float=0.2453
    ConstraintWeights_3:float=0.1411
    ConstraintWeights_4:float=0.0671
    ConstraintWeights_5:float=0.0068
   ) -> (Pose) {
    boneToModify {boneName=base_cord_l_01}
    constraintSetup [
     {targetBone {boneName=spine_04}}
     {targetBone {boneName=spine_02}}
     {targetBone {boneName=spine_03}}
     {targetBone {boneName=spine_05}}
     {targetBone {boneName=spine_01}}
     {targetBone {boneName=pelvis}}
    ]
    constraintWeights [1,1,1,1,1,1]
   }
   n26 = "Constraint - base_cord_r_01 (PPPPPP)"!(
    ComponentPose=@n25
    Alpha=1
    ConstraintWeights_0=0.2862
    ConstraintWeights_1:float=0.2486
    ConstraintWeights_2:float=0.2457
    ConstraintWeights_3:float=0.1461
    ConstraintWeights_4:float=0.0659
    ConstraintWeights_5:float=0.0075
   ) -> (Pose) {
    boneToModify {boneName=base_cord_r_01}
    constraintSetup [
     {targetBone {boneName=spine_04}}
     {targetBone {boneName=spine_02}}
     {targetBone {boneName=spine_03}}
     {targetBone {boneName=spine_05}}
     {targetBone {boneName=spine_01}}
     {targetBone {boneName=pelvis}}
    ]
    constraintWeights [1,1,1,1,1,1]
   }
   n27 = Anim Dynamics - Bone: dyn_bow_l(ComponentPose=@n26, Alpha=1, ExternalForce="0,0,0") {
    angularDampingOverride=1
    angularSpringConstant=16000
    bAngularSpring=true
    bOverrideAngularDamping=true
    bUseGravityOverride=true
    gravityScale=0.35
    simulationSpace=RootRelative
    boundBone {boneName=dyn_bow_l}
    componentAppliedLinearAccClamp=<6000 6000 6000>
    componentLinearAccScale=<0.8 0.8 0.8>
    componentLinearVelScale=<0.9 0.9 0.9>
    gravityOverride {z=-980}
    physicsBodyDefinitions [{
      boundBone {boneName=dyn_bow_l}
      boxExtents=<20 20 20>
      constraintSetup {angularLimitsMax=<0 8 0>, angularLimitsMin=<0 -20 -27>, angularTarget=<1 0 0>}
      localJointOffset=<5 0 0>
     }]
   }
   n28 = Anim Dynamics - Bone: dyn_bow_r(ComponentPose=@n27, Alpha=1, ExternalForce="0,0,0") {
    angularDampingOverride=1
    angularSpringConstant=18000
    bAngularSpring=true
    bOverrideAngularDamping=true
    bUseGravityOverride=true
    gravityScale=0.35
    simulationSpace=RootRelative
    boundBone {boneName=dyn_bow_r}
    componentAppliedLinearAccClamp=<6000 6000 6000>
    componentLinearAccScale=<0.7 0.7 0.7>
    componentLinearVelScale=<0.7 0.7 0.7>
    gravityOverride {z=-960}
    physicsBodyDefinitions [{
      boundBone {boneName=dyn_bow_r}
      boxExtents=<19 19 19>
      constraintSetup {angularLimitsMax=<0 27 0>, angularLimitsMin=<0 -10 -29>, angularTarget=<1 0 0>}
      localJointOffset=<5.3 0 0>
     }]
   }
   n29 = Anim Dynamics - Bone: dyn_cord_r_01(ComponentPose=@n28, Alpha=1, ExternalForce="0,0,0") {
    angularDampingOverride=1
    angularSpringConstant=17500
    bAngularSpring=true
    bOverrideAngularDamping=true
    bUseGravityOverride=true
    gravityScale=0.35
    simulationSpace=RootRelative
    boundBone {boneName=dyn_cord_r_01}
    componentAppliedLinearAccClamp=<6000 6000 6000>
    componentLinearAccScale=<0.8 0.8 0.8>
    componentLinearVelScale=<0.8 0.8 0.8>
    gravityOverride {z=-970}
    physicsBodyDefinitions [{
      boundBone {boneName=dyn_cord_r_01}
      boxExtents=<28 28 28>
      constraintSetup {angularLimitsMax=<0 15 2>, angularLimitsMin=<0 -15 -18>, angularTarget=<1 0 0>}
      localJointOffset=<3 0 0>
     }]
   }
   n30 = Anim Dynamics - Bone: dyn_cord_r_02(ComponentPose=@n29, Alpha=1, ExternalForce="0,0,0") {
    angularDampingOverride=1
    angularSpringConstant=17500
    bAngularSpring=true
    bOverrideAngularDamping=true
    bUseGravityOverride=true
    gravityScale=0.35
    simulationSpace=RootRelative
    boundBone {boneName=dyn_cord_r_02}
    componentAppliedLinearAccClamp=<6000 6000 6000>
    componentLinearAccScale=<0.8 0.8 0.8>
    componentLinearVelScale=<0.8 0.8 0.8>
    gravityOverride {z=-970}
    physicsBodyDefinitions [{
      boundBone {boneName=dyn_cord_r_02}
      boxExtents=<28 28 28>
      constraintSetup {angularLimitsMax=<0 15 2>, angularLimitsMin=<0 -15 -18>, angularTarget=<1 0 0>}
      localJointOffset=<3 0 0>
     }]
   }
   n31 = Anim Dynamics - Bone: dyn_cord_l_01(ComponentPose=@n30, Alpha=1, ExternalForce="0,0,0") {
    angularDampingOverride=1
    angularSpringConstant=19000
    bAngularSpring=true
    bOverrideAngularDamping=true
    bUseGravityOverride=true
    gravityScale=0.35
    simulationSpace=RootRelative
    boundBone {boneName=dyn_cord_l_01}
    componentAppliedLinearAccClamp=<6000 6000 6000>
    componentLinearAccScale=<1 1 1>
    componentLinearVelScale=<1 1 1>
    gravityOverride {z=-980}
    physicsBodyDefinitions [{
      boundBone {boneName=dyn_cord_l_01}
      boxExtents=<27.5 27.5 27.5>
      constraintSetup {angularLimitsMax=<0 15 2>, angularLimitsMin=<0 -15 -18>, angularTarget=<1 0 0>}
      localJointOffset=<4.25 0 0>
     }]
   }
   n32 = Anim Dynamics - Bone: dyn_cord_l_02(ComponentPose=@n31, Alpha=1, ExternalForce="0,0,0") {
    angularDampingOverride=1
    angularSpringConstant=19000
    bAngularSpring=true
    bOverrideAngularDamping=true
    bUseGravityOverride=true
    gravityScale=0.35
    simulationSpace=RootRelative
    boundBone {boneName=dyn_cord_l_02}
    componentAppliedLinearAccClamp=<6000 6000 6000>
    componentLinearAccScale=<1 1 1>
    componentLinearVelScale=<1 1 1>
    gravityOverride {z=-980}
    physicsBodyDefinitions [{
      boundBone {boneName=dyn_cord_l_02}
      boxExtents=<27.5 27.5 27.5>
      constraintSetup {angularLimitsMax=<0 15 2>, angularLimitsMin=<0 -15 -18>, angularTarget=<1 0 0>}
      localJointOffset=<4.25 0 0>
     }]
   }
   n33 = "Transform (Modify) Bone - Bone: base_bow_l"!(ComponentPose=@n32, bAlphaBoolEnabled:bool=True, Translation="0,0,0", Rotation="10,-17,15", Scale="1,1,1") -> (Pose) {
    alphaInputType=Bool
    rotationMode=BMM_Additive
    rotationSpace=BCS_BoneSpace
    alphaBoolBlend {blendInTime=0.27, blendOutTime=0.33}
    boneToModify {boneName=base_bow_l}
    rotation=<rot 10 -17 15>
   }
   n34 = "Transform (Modify) Bone - Bone: base_bow_r"!(ComponentPose=@n33, bAlphaBoolEnabled:bool=True, Translation="0,0,0", Rotation="-26,-27,1", Scale="1,1,1") -> (Pose) {
    alphaInputType=Bool
    rotationMode=BMM_Additive
    rotationSpace=BCS_BoneSpace
    alphaBoolBlend {blendInTime=0.35, blendOption=Sinusoidal, blendOutTime=0.47}
    boneToModify {boneName=base_bow_r}
    rotation=<rot -26 -27 1>
   }
   n35 = "Transform (Modify) Bone - Bone: base_bow_l"(ComponentPose=@n34, bAlphaBoolEnabled=True, Translation="0,-1,0", Rotation="0,0,0", Scale="1,1,1") {
    alphaInputType=Bool
    rotationMode=BMM_Additive
    rotationSpace=BCS_BoneSpace
    translationMode=BMM_Additive
    translationSpace=BCS_BoneSpace
    alphaBoolBlend {blendInTime=0.27, blendOutTime=0.33}
    boneToModify {boneName=base_bow_l}
    translation {y=-1}
   }
   n36 = "Transform (Modify) Bone - Bone: base_cord_l_01"!(ComponentPose=@n35, bAlphaBoolEnabled:bool=True, Translation="0,-1,0", Rotation="0,0,0", Scale="1,1,1") -> (Pose) {
    alphaInputType=Bool
    rotationMode=BMM_Additive
    rotationSpace=BCS_BoneSpace
    translationMode=BMM_Additive
    translationSpace=BCS_BoneSpace
    alphaBoolBlend {blendInTime=0.35, blendOutTime=0.45}
    boneToModify {boneName=base_cord_l_01}
    translation {y=-1}
   }
   n37 = "Transform (Modify) Bone - Bone: base_cord_r_01"!(ComponentPose=@n36, bAlphaBoolEnabled:bool=True, Translation="0,-1,0", Rotation="0,0,0", Scale="1,1,1") -> (Pose) {
    alphaInputType=Bool
    rotationMode=BMM_Additive
    rotationSpace=BCS_BoneSpace
    translationMode=BMM_Additive
    translationSpace=BCS_BoneSpace
    alphaBoolBlend {blendInTime=0.35, blendOutTime=0.45}
    boneToModify {boneName=base_cord_r_01}
    translation {y=-1}
   }
   n38 = Anim Dynamics - Bone: dyn_necklace(ComponentPose=@n37, Alpha=1, ExternalForce="0,0,0") {
    angularDampingOverride=1
    angularSpringConstant=18000
    bAngularSpring=true
    bOverrideAngularDamping=true
    bUseGravityOverride=true
    gravityScale=0.35
    simulationSpace=RootRelative
    boundBone {boneName=dyn_necklace}
    componentAppliedLinearAccClamp=<6000 6000 6000>
    componentLinearAccScale=<0.6 0.6 0.6>
    componentLinearVelScale=<0.6 0.6 0.6>
    gravityOverride {z=-980}
    physicsBodyDefinitions [{
      boundBone {boneName=dyn_necklace}
      boxExtents=<28 26 26>
      constraintSetup {angularLimitsMax=<3 0 0>, angularLimitsMin=<-4 0 -18>, angularTarget=<1 0 0>}
      localJointOffset=<4 0 0>
     }]
   }
   n39 = "Transform (Modify) Bone - Bone: base_necklace"(ComponentPose=@n38, Alpha=1, Translation="0.4,0,0", Rotation="0,0,0", Scale="1,1,1") {translationMode=BMM_Additive, translationSpace=BCS_BoneSpace, boneToModify {boneName=base_necklace}, translation {x=0.4}}
   Save cached pose 'allDynCache'(Pose=Component To Local(ComponentPose=@n39)) {cacheName=allDynCache}
  ]}
 EventGraph {nodes [
   Event BlueprintUpdateAnimation()
   > n1 = "Gravity Override State Function (Parameter Asset) (Nativized)"(
    CharacterPartAnimInstance=Self-Reference()
    GameTimeInSeconds=GetGameTimeInSeconds()
    DebugJointName=spine_05
    Parameters=/BRCosmetics/Characters/Player/Female/Medium/Bodies/F_MED_LivelySugar_Shore/Meshes/F_MED_LivelySugar_Shore_Parameters
   )
   > Set Gravity Override(Gravity Override=@n1)
   > n2 = Sequence()
   n3 = Get bIsSkyDiving()
   n4 = Get bIsFalling()
   n5 = Get bIsJumping()
   > n6 = Branch(Condition=AND Boolean!(A=OR Boolean(A=@n4, B=@n5), B=NOT Boolean(A=@n3), C:bool=true) -> (ReturnValue))
   > n7 = Branch(Condition=Get bIsTacticalSprinting())
   > Set isFallingNotMoving(isFallingNotMoving:double=0.05) -> (Output_Get:double)
   Set isFallingNotMoving(execute=@n7.else, isFallingNotMoving=0.7)
   Set isFallingNotMoving(execute=@n6.else, isFallingNotMoving=0.3)
   "Set Parachuting shoulder value "(execute=@n2.then_1, "Parachuting shoulder value ":double="To Float (Boolean)"(InBool=Get bIsParachuteOpened())) -> (Output_Get:double)
   NOT Boolean(A=Get bIsMoving2D())
  ]}
]
variables [
 Gravity Override {type=Vector}
 isFallingNotMoving {type=number}
 JumpFalling {type=boolean}
 "Parachuting shoulder value " {type=number}
 Sim Space Settings {
  type=AnimPhysSimSpaceSettings
  default {maxAngularAcceleration=45, maxAngularVelocity=8, simSpaceAngularAlpha=0.25, externalAngularVelocity=<0 0 0>}
 }
]
