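# Physics asset of collision shapes for capes on the LivelySugar Syrup female medium character body, as its CapeColliders name indicates.
# Capsules cover pelvis, spine, neck and clavicles, with a sphere for the head and overlapping capsules on spine_02 and spine_05.
# Extra capsules follow the dynamic coat back, handkerchief and holster bones.
# Arm, leg and foot bodies hold no shapes, and no constraints join any bodies.
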PhysicsAsset /BRCosmetics/Characters/Player/Female/Medium/Bodies/F_MED_LivelySugar_Syrup/Meshes/F_MED_LivelySugar_Syrup_CapeColliders
bodies {
 calf_l []
 calf_r []
 clavicle_l [{shapeType=Capsule, length=9.0192, radius=9.8871, center=<9.4266 -1.1806 -2.4977>, rotation=<rot 69.3297 1.4476 176.9462>}]
 clavicle_r [{shapeType=Capsule, length=9.0192, radius=9.8871, center=<-9.4266 1.1806 2.4977>, rotation=<rot 69.3297 1.4476 -3.0538>}]
 dyn_coatBack [{
   shapeType=Capsule
   length=18.9767
   radius=7.6243
   shapeName=dyn_coatBack_capsule
   center=<0.3722 1.8251 -0.8055>
   rotation=<rot 0 75.3544 176.228>
  }]
 dyn_handkerchiefBottom_01 [{
   shapeType=Capsule
   length=12.2908
   radius=3.8766
   shapeName=dyn_handkerchiefBottom_01_capsule
   center=<11.5059 -0.0822 0.4729>
   rotation=<rot 0 -85.4079 -87.3901>
  }]
 dyn_handkerchiefTop_01 [{
   shapeType=Capsule
   length=7.5233
   radius=2.6434
   shapeName=dyn_handkerchiefTop_01_capsule
   center=<7.4199 0.4492 0.4278>
   rotation=<rot 0 92.5148 -89.6167>
  }]
 dyn_holster [{
   shapeType=Capsule
   length=12.5673
   radius=7.5668
   shapeName=dyn_holster_capsule
   center=<0.0211 -2.2702 4.0659>
   rotation=<rot 0 -96.4867 -86.9477>
  }]
 foot_l []
 foot_r []
 head [{shapeType=Sphere, radius=17.3125, center=<11.028 -0.174 0>}]
 lowerarm_l []
 lowerarm_r []
 neck_02 [{shapeType=Capsule, length=13.4375, radius=7.3175, center=<-0.2642 -0.3408 0>, rotation=<rot 90 -171.1193 -175.3978>}]
 pelvis [{shapeType=Capsule, length=8.6456, radius=14.5388, center=<-2.4887 0.1881 0>}]
 spine_01 [{shapeType=Capsule, length=7.9438, radius=13.5586, center=<-1.7965 0.1884 0>}]
 spine_02 [
  {shapeType=Capsule, length=7.3264, radius=13.6322, center=<-1.711 -0.1851 0.0001>}
  {shapeType=Capsule, length=29.3956, radius=11.5807, center=<2.5024 0.4992 3.9825>, rotation=<rot 90 0.0049 0.0049>}
  {shapeType=Capsule, length=29.875, radius=11.1013, center=<2.4704 0.4991 -3.9825>, rotation=<rot 90 -0.0396 -0.0396>}
 ]
 spine_03 [{shapeType=Capsule, length=8.2107, radius=13.7031, center=<0.1445 0.2396 0.0001>}]
 spine_04 [{shapeType=Capsule, length=9.4404, radius=14.1325, center=<2.766 0.7765 0.0002>}]
 spine_05 [
  {shapeType=Capsule, length=8.0226, radius=15.695, center=<-2.2795 1.8109 0>}
  {shapeType=Capsule, length=13.9518, radius=13.9425, center=<-0.2916 0.4452 0>}
 ]
 thigh_l []
 thigh_r []
 upperarm_l []
 upperarm_r []
}
constraints []
